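# Physics asset for the HonestWasp axe pickaxe cosmetic, defining physics bodies for a chain on the weapon mesh.
# Single capsules cover a small base_chain body and four dyn_chain bones, each constrained to the one before it in a linear chain.
# The first two joints bend mainly on opposite swing axes, the last two swing equally on both with a wider range at the tip, and twist stays tight throughout.
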PhysicsAsset /BRCosmetics/Weapons/FORT_Melee/Pickaxe_HonestWaspUniversal/Meshes/HonestWasp_Axe_Physics
bodies {
 base_chain [{shapeType=Capsule, length=0.505, radius=0.505, shapeName=base_chain_capsule, rotation=<rot 0 -90 0>}]
 dyn_chain_01 [{shapeType=Capsule, length=0.5, radius=1.5, shapeName=dyn_chain_01_capsule, center=<1.3746 0 0>, rotation=<rot 90 0 0>}]
 dyn_chain_02 [{shapeType=Capsule, length=3.5, radius=1.5, shapeName=dyn_chain_02_capsule, center=<3.317 0 0>, rotation=<rot 90 0 0>}]
 dyn_chain_03 [{shapeType=Capsule, length=2.5, radius=1.5, shapeName=dyn_chain_03_capsule, center=<3.317 0 0>, rotation=<rot 90 0 0>}]
 dyn_chain_04 [{shapeType=Capsule, length=2.5, radius=1.5, shapeName=dyn_chain_04_capsule, center=<3.317 0 0>, rotation=<rot 90 0 0>}]
}
constraints [
 {bone1Name=dyn_chain_01, bone2Name=base_chain, swing1LimitDegrees=60, swing2LimitDegrees=5, twistLimitDegrees=1}
 {bone1Name=dyn_chain_02, bone2Name=dyn_chain_01, swing1LimitDegrees=5, swing2LimitDegrees=60, twistLimitDegrees=5}
 {bone1Name=dyn_chain_03, bone2Name=dyn_chain_02, swing1LimitDegrees=30, swing2LimitDegrees=30, twistLimitDegrees=10}
 {bone1Name=dyn_chain_04, bone2Name=dyn_chain_03, swing1LimitDegrees=45, swing2LimitDegrees=45, twistLimitDegrees=10}
]
